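# Sound Cue for musical ambient one-shot spots at the Mount Olympus point of interest.
# A single Random node picks among ten equally weighted Olympus one-shot waves.
# It plays quietly in the music sound class with a musical no-reverb one-shot attenuation.
# A concurrency override allows one instance at a time and blocks new plays while it sounds.
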SoundCue /Game/Sounds/Helios/POI/MountOlympus/Spots/Musical/FN_S29_AmbientOneShots_Olympus_9_v3_Cue
root=Random_0
nodes [
 {
  id=Random_0
  children [WavePlayer_0,WavePlayer_1,WavePlayer_2,WavePlayer_3,WavePlayer_4,WavePlayer_5,WavePlayer_6,WavePlayer_7,WavePlayer_8,WavePlayer_9]
  properties {weights [1,1,1,1,1,1,1,1,1,1]}
 }
 {id=WavePlayer_0, properties {soundWaveAssetPtr=&/FN_S29_AmbientOneShots_Olympus_9_v3}}
 {id=WavePlayer_1, properties {soundWaveAssetPtr=&/FN_S29_AmbientOneShots_Olympus_8_v3}}
 {id=WavePlayer_2, properties {soundWaveAssetPtr=&/FN_S29_AmbientOneShots_Olympus_7_v3}}
 {id=WavePlayer_3, properties {soundWaveAssetPtr=&/FN_S29_AmbientOneShots_Olympus_6_v3}}
 {id=WavePlayer_4, properties {soundWaveAssetPtr=&/FN_S29_AmbientOneShots_Olympus_5_v3}}
 {id=WavePlayer_5, properties {soundWaveAssetPtr=&/FN_S29_AmbientOneShots_Olympus_4_v3}}
 {id=WavePlayer_6, properties {soundWaveAssetPtr=&/FN_S29_AmbientOneShots_Olympus_3_v3}}
 {id=WavePlayer_7, properties {soundWaveAssetPtr=&/FN_S29_AmbientOneShots_Olympus_2_v3}}
 {id=WavePlayer_8, properties {soundWaveAssetPtr=&/FN_S29_AmbientOneShots_Olympus_1_v3}}
 {id=WavePlayer_9, properties {soundWaveAssetPtr=&/FN_S29_AmbientOneShots_Olympus_10_v3}}
]
properties {
 attenuationSettings=/Game/Sounds/Attenuation/Fort_Ambient_Athena_Musical_Oneshot_NoVerb_Att
 bOverrideConcurrency=true
 maxDistance=2097152
 soundClassObject=/Game/Sounds/FortSoundClassesAndModes/Fort_Music
 volumeMultiplier=0.15
 concurrencyOverrides {maxCount=1, resolutionRule=PreventNew, voiceStealReleaseTime=2}
}
